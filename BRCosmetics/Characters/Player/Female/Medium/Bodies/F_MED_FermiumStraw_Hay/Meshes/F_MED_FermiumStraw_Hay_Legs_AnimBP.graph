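# Post-process style AnimBlueprint for the FermiumStraw Hay legs cosmetic, adding secondary motion to cans, flaps, belt and zipper.
# The AnimGraph copies the parent mesh pose through a Control Rig, then chains per-bone Anim Dynamics nodes.
# The EventGraph sets dynamics alpha variables from skydiving and front-end state.
# Outside the front end, it also feeds foot and spine motion, plus skydive or glider wind, into material-driven WPO cloth simulation.
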AnimBlueprint /BRCosmetics/Characters/Player/Female/Medium/Bodies/F_MED_FermiumStraw_Hay/Meshes/F_MED_FermiumStraw_Hay_Legs_AnimBP
target_skeleton=&/F_MED_FermiumStraw_Hay_Skeleton
graphs [
 AnimGraph {nodes [
   Output Pose(Result=Use cached pose 'AnimDynamicsCache'()) {layerGroup=None}
   n1 = Control Rig!(Source=Copy Pose From Mesh() {bUseAttachedParent=true}, Alpha, IsSky:bool=false, isFrontEnd:bool=false) -> (Pose) {
    lODThreshold=3
    controlRigAssetReference {controlRigAsset=&/F_MED_FermiumStraw_Hay_Legs_CtrlRig}
   }
   Save cached pose 'ControlRigCache'(Pose=@n1) {cacheName=ControlRigCache}
   n2 = Anim Dynamics - Bone: dyn_can_01!(ComponentPose=Local To Component(LocalPose=Use cached pose 'ControlRigCache'()), Alpha=0.5, GravityOverride:Vector="0,0,-980") -> (Pose) {
    angularDampingOverride=1
    angularSpringConstant=9000
    bAngularSpring=true
    bOverrideAngularDamping=true
    bUseGravityOverride=true
    gravityScale=0.35
    simulationSpace=RootRelative
    boundBone {boneName=dyn_can_01}
    componentAppliedLinearAccClamp=<3000 3000 3000>
    componentLinearAccScale=<0.4 0.4 0.4>
    componentLinearVelScale=<0.2 0.2 0.2>
    gravityOverride {z=-980}
    physicsBodyDefinitions [{
      boundBone {boneName=dyn_can_01}
      boxExtents=<20 20 20>
      constraintSetup {angularLimitsMax=<0 40 40>, angularLimitsMin=<0 -40 -90>, angularTarget=<1 0 0>}
      localJointOffset=<2 0 0>
     }]
    simSpaceSettings {maxAngularAcceleration=20, maxAngularVelocity=10, simSpaceAngularAlpha=1}
   }
   n3 = Anim Dynamics - Bone: dyn_can_02!(ComponentPose=@n2, Alpha=0.5, GravityOverride:Vector="0,0,-980") -> (Pose) {
    angularDampingOverride=1
    angularSpringConstant=9000
    bAngularSpring=true
    bOverrideAngularDamping=true
    bUseGravityOverride=true
    gravityScale=0.35
    simulationSpace=RootRelative
    boundBone {boneName=dyn_can_02}
    componentAppliedLinearAccClamp=<3000 3000 3000>
    componentLinearAccScale=<0.4 0.4 0.4>
    componentLinearVelScale=<0.2 0.2 0.2>
    gravityOverride {z=-980}
    physicsBodyDefinitions [{
      boundBone {boneName=dyn_can_02}
      boxExtents=<20 20 20>
      constraintSetup {angularLimitsMax=<0 10 7>, angularLimitsMin=<0 -10 -25>, angularTarget=<1 0 0>}
      localJointOffset=<6 0 0>
     }]
    simSpaceSettings {maxAngularAcceleration=20, maxAngularVelocity=10, simSpaceAngularAlpha=1}
   }
   n4 = Anim Dynamics - Bone: dyn_can_03!(ComponentPose=@n3, Alpha=0.5, GravityOverride:Vector="0,0,-980") -> (Pose) {
    angularDampingOverride=1
    angularSpringConstant=9000
    bAngularSpring=true
    bOverrideAngularDamping=true
    bUseGravityOverride=true
    gravityScale=0.35
    simulationSpace=RootRelative
    boundBone {boneName=dyn_can_03}
    componentAppliedLinearAccClamp=<3000 3000 3000>
    componentLinearAccScale=<0.4 0.4 0.4>
    componentLinearVelScale=<0.2 0.2 0.2>
    gravityOverride {z=-980}
    physicsBodyDefinitions [{
      boundBone {boneName=dyn_can_03}
      boxExtents=<20 20 20>
      constraintSetup {angularLimitsMax=<0 40 70>, angularLimitsMin=<0 -40 -70>, angularTarget=<1 0 0>}
      localJointOffset=<2 0 0>
     }]
    simSpaceSettings {maxAngularAcceleration=20, maxAngularVelocity=10, simSpaceAngularAlpha=1}
   }
   n5 = Anim Dynamics - Bone: dyn_legFlap_a_l!(ComponentPose=@n4, Alpha=1, GravityOverride:Vector="0,0,-980") -> (Pose) {
    angularDampingOverride=1
    angularSpringConstant=9000
    bAngularSpring=true
    bOverrideAngularDamping=true
    bUseGravityOverride=true
    gravityScale=0.35
    simulationSpace=RootRelative
    boundBone {boneName=dyn_legFlap_a_l}
    componentAppliedLinearAccClamp=<3000 3000 3000>
    componentLinearAccScale=<0.4 0.4 0.4>
    componentLinearVelScale=<0.4 0.4 0.4>
    gravityOverride {z=-980}
    physicsBodyDefinitions [{
      boundBone {boneName=dyn_legFlap_a_l}
      boxExtents=<20 20 20>
      constraintSetup {angularLimitsMax=<0 20 20>, angularLimitsMin=<0 -20 -70>, angularTarget=<1 0 0>}
      localJointOffset=<6 0 0>
     }]
    simSpaceSettings {maxAngularAcceleration=20, maxAngularVelocity=10, simSpaceAngularAlpha=1}
   }
   n6 = Anim Dynamics - Bone: dyn_legFlap_b_l!(ComponentPose=@n5, Alpha=1, GravityOverride:Vector="0,0,-980") -> (Pose) {
    angularDampingOverride=1
    angularSpringConstant=9000
    bAngularSpring=true
    bOverrideAngularDamping=true
    bUseGravityOverride=true
    gravityScale=0.35
    simulationSpace=RootRelative
    boundBone {boneName=dyn_legFlap_b_l}
    componentAppliedLinearAccClamp=<3000 3000 3000>
    componentLinearAccScale=<0.4 0.4 0.4>
    componentLinearVelScale=<0.4 0.4 0.4>
    gravityOverride {z=-980}
    physicsBodyDefinitions [{
      boundBone {boneName=dyn_legFlap_b_l}
      boxExtents=<20 20 20>
      constraintSetup {angularLimitsMax=<0 20 -10>, angularLimitsMin=<0 -20 -70>, angularTarget=<1 0 0>}
      localJointOffset=<6 0 0>
     }]
    simSpaceSettings {maxAngularAcceleration=20, maxAngularVelocity=10, simSpaceAngularAlpha=1}
   }
   n7 = Anim Dynamics - Bone: dyn_regFlap_a_r!(ComponentPose=@n6, Alpha=1, GravityOverride:Vector="0,0,-980") -> (Pose) {
    angularDampingOverride=1
    angularSpringConstant=9000
    bAngularSpring=true
    bOverrideAngularDamping=true
    bUseGravityOverride=true
    gravityScale=0.35
    simulationSpace=RootRelative
    boundBone {boneName=dyn_regFlap_a_r}
    componentAppliedLinearAccClamp=<3000 3000 3000>
    componentLinearAccScale=<0.4 0.4 0.4>
    componentLinearVelScale=<0.4 0.4 0.4>
    gravityOverride {z=-980}
    physicsBodyDefinitions [{
      boundBone {boneName=dyn_regFlap_a_r}
      boxExtents=<20 20 20>
      constraintSetup {angularLimitsMax=<0 20 40>, angularLimitsMin=<0 -20 -70>, angularTarget=<1 0 0>}
      localJointOffset=<-6 0 0>
     }]
    simSpaceSettings {maxAngularAcceleration=20, maxAngularVelocity=10, simSpaceAngularAlpha=1}
   }
   n8 = Anim Dynamics - Bone: dyn_regFlap_b_r!(ComponentPose=@n7, Alpha=1, GravityOverride:Vector="0,0,-980") -> (Pose) {
    angularDampingOverride=1
    angularSpringConstant=9000
    bAngularSpring=true
    bOverrideAngularDamping=true
    bUseGravityOverride=true
    gravityScale=0.35
    simulationSpace=RootRelative
    boundBone {boneName=dyn_regFlap_b_r}
    componentAppliedLinearAccClamp=<3000 3000 3000>
    componentLinearAccScale=<0.4 0.4 0.4>
    componentLinearVelScale=<0.4 0.4 0.4>
    gravityOverride {z=-980}
    physicsBodyDefinitions [{
      boundBone {boneName=dyn_regFlap_b_r}
      boxExtents=<20 20 20>
      constraintSetup {angularLimitsMax=<0 20 80>, angularLimitsMin=<0 -20 -70>, angularTarget=<1 0 0>}
      localJointOffset=<-12 0 0>
     }]
    simSpaceSettings {maxAngularAcceleration=20, maxAngularVelocity=10, simSpaceAngularAlpha=1}
   }
   n9 = Anim Dynamics - Bone: dyn_beltZipper!(ComponentPose=@n8, Alpha=1, GravityOverride:Vector="0,0,-980") -> (Pose) {
    angularDampingOverride=1
    angularSpringConstant=9000
    bAngularSpring=true
    bOverrideAngularDamping=true
    bUseGravityOverride=true
    gravityScale=0.35
    simulationSpace=RootRelative
    boundBone {boneName=dyn_beltZipper}
    componentAppliedLinearAccClamp=<3000 3000 3000>
    componentLinearAccScale=<0.4 0.4 0.4>
    componentLinearVelScale=<0.4 0.4 0.4>
    gravityOverride {z=-980}
    physicsBodyDefinitions [{
      boundBone {boneName=dyn_beltZipper}
      boxExtents=<20 20 20>
      constraintSetup {angularLimitsMax=<0 20 0>, angularLimitsMin=<0 -20 -70>, angularTarget=<1 0 0>}
      localJointOffset=<6 0 0>
     }]
    simSpaceSettings {maxAngularAcceleration=20, maxAngularVelocity=10, simSpaceAngularAlpha=1}
   }
   n10 = Anim Dynamics - Bone: dyn_belt_03!(ComponentPose=@n9, Alpha=0.8, GravityOverride:Vector="0,0,-980") -> (Pose) {
    angularDampingOverride=1
    angularSpringConstant=20000
    bAngularSpring=true
    bOverrideAngularDamping=true
    bUseGravityOverride=true
    gravityScale=0.35
    simulationSpace=RootRelative
    boundBone {boneName=dyn_belt_03}
    componentAppliedLinearAccClamp=<3000 3000 3000>
    componentLinearAccScale=<1 1 1>
    componentLinearVelScale=<0.4 0.4 0.4>
    gravityOverride {z=-980}
    physicsBodyDefinitions [{
      boundBone {boneName=dyn_belt_03}
      boxExtents=<30 30 30>
      constraintSetup {angularLimitsMax=<0 0 0>, angularLimitsMin=<0 0 -90>, angularTarget=<1 0 0>}
      localJointOffset=<4 0 0>
     }]
    simSpaceSettings {maxAngularAcceleration=20, maxAngularVelocity=10, simSpaceAngularAlpha=1}
   }
   Save cached pose 'AnimDynamicsCache'(Pose=Component To Local(ComponentPose=@n10)) {cacheName=AnimDynamicsCache}
  ]}
 EventGraph {nodes [
   Event BlueprintInitializeAnimation()
   n1 = TryGetPawnOwner()
   > Is Valid(InputObject=@n1)
   > n2 = Cast To FortPlayerPawnAthena(Object=@n1)
   > Set PlayerPawn(PlayerPawn:Fort Player Pawn Athena Object Reference=@n2) -> (Output_Get:Fort Player Pawn Athena Object Reference)
   Event OnParachuteOpened()
   > n3 = Switch on EGliderType(Selection=Surfing)
   Set isSurfing?(execute=@n3.Surfing, isSurfing?:bool=true) -> (Output_Get:bool)
   Event OnParachuteClosed()
   > Set isSurfing?(isSurfing?=false)
   n4 = Event BlueprintUpdateAnimation()
   > K2Node_Composite_1 = WPO_setup!(InputPin:Exec, InputPin2:float=@n4.DeltaTimeX) -> (then_1:Exec) {is=title K2Node_Composite}
   n5 = Get isSurfing?() -> (isSurfing?:bool)
   n6 = Get bIsSkyDiving()
   > Set InSky?(InSky?:bool=AND Boolean(A=@n6, B=NOT Boolean(A=@n5))) -> (Output_Get:bool)
   > n7 = Sequence()
   > n8 = Branch(Condition=@n6)
   > Set BeltDynAlpha(BeltDynAlpha:double=0) -> (Output_Get:double)
   Set BeltDynAlpha(execute=@n8.else, BeltDynAlpha=0.8)
   n9 = Branch(execute=@n7.then_1, Condition=Get bIsInFrontEnd())
   > Set AlphaCloth(AlphaCloth:double=0.5) -> (Output_Get:double)
   > Set AlphaCans(AlphaCans:double=0.25) -> (Output_Get:double)
   > Set BeltDynAlpha(BeltDynAlpha=0)
   Set AlphaCloth(execute=@n9.else, AlphaCloth=1)
   > Set AlphaCans(AlphaCans=0.5)
   Event OnCharacterCustomizationLoaded()
   > n10 = Set SkelMesh(SkelMesh=GetPartSkeletalMeshComponent())
   > n11 = Get Dynamic Material Instance(SkeletalMeshComponent=@n10)
   > Set Mat Instance Dynamic(Mat Instance Dynamic=@n11)
   > n12 = GetOrCreateDynamicOverlayMaterial(MeshComponent=@n10, MaterialName=None)
   > Set Overlay Mat Instance Dynamic(Overlay Mat Instance Dynamic=@n12)
  ]}
 WPO_setup {
  path=EventGraph.K2Node_Composite_1.WPO_setup
  nodes [
   n1 = Inputs!() -> (InputPin:Exec, InputPin2:float=~)
   > n2 = Sequence()
   > Is Valid(InputObject=Get PlayerPawn() -> (PlayerPawn:Fort Player Pawn Athena Object Reference))
   > n3 = Branch(Condition=Get bIsInFrontEnd())
   Set WPO Globals Function(
    execute=@n3.else
    Material=Get Mat Instance Dynamic()
    Overlay Material=Get Overlay Mat Instance Dynamic()
    Skeletal Mesh=Get SkelMesh()
    "Local Noise Transform  Location"=Get Local Noise Transform Location() -> (Local Noise Transform Location:Vector)
    Local Noise Offset=Get Local Noise Offset()
   )
   > Set WPO Regional Function!(
    execute
    Material:Material Instance Dynamic Object Reference=Get Mat Instance Dynamic()
    Overlay Material:Material Instance Dynamic Object Reference=Get Overlay Mat Instance Dynamic()
    Skeletal Mesh:Skeletal Mesh Component Object Reference=Get SkelMesh()
    Velocity Transform:Name=spine_05
    Velocity Interp Speed:double=Get Velocity Interp Speed()
    Collision Transform:Name=root
    Collision Transform Axis:EAxis Enum
    Invert Collision Transform Axis:bool=false
    Delta Time X:double=@n1
    Previous Velocity Transform Location:"Vector (by ref)"=Get R1 Velocity Transform Location()
    Previous Velocity Transform Velocity:"Vector (by ref)"=Get R1 Velocity Transform Velocity()
    Region:int=1
   ) -> (then)
   > Set WPO Regional Function(
    Material=Get Mat Instance Dynamic()
    Overlay Material=Get Overlay Mat Instance Dynamic()
    Skeletal Mesh=Get SkelMesh()
    Velocity Transform=foot_l
    Velocity Interp Speed=Get Velocity Interp Speed()
    Collision Transform=calf_l
    Invert Collision Transform Axis=true
    Delta Time X=@n1
    Previous Velocity Transform Location=Get R1 Velocity Transform Location()
    Previous Velocity Transform Velocity=Get R1 Velocity Transform Velocity()
    Region=1
   )
   > Set WPO Regional Function(
    Material=Get Mat Instance Dynamic()
    Overlay Material=Get Overlay Mat Instance Dynamic()
    Skeletal Mesh=Get SkelMesh()
    Velocity Transform=foot_r
    Velocity Interp Speed=Get Velocity Interp Speed()
    Collision Transform=calf_r
    Invert Collision Transform Axis=true
    Delta Time X=@n1
    Previous Velocity Transform Location=Get R2 Velocity Transform Location()
    Previous Velocity Transform Velocity=Get R2 Velocity Transform Velocity()
    Region=2
   )
   > n4 = Branch(Condition=OR Boolean(A=Get bIsSkydiveFloating(), B=Get bIsParachuteOpened()))
   > Set WPO Pho Wind Intensity(WPO Pho Wind Intensity=0.8)
   > n5 = Set WPO Pho Turbulence(WPO Pho Turbulence:double=4.2) -> (Output_Get:double)
   SetScalarParameterValue(
    execute=[@n7.then, @n5.then]
    self=[Get Overlay Mat Instance Dynamic(), Get Mat Instance Dynamic()]
    ParameterName=ClothSim_WindStrength
    Value=Get WPO Pho Wind Intensity()
   )
   n6 = Get WPO Pho Turbulence() -> (WPO Pho Turbulence:double)
   > SetScalarParameterValue(self=[Get Overlay Mat Instance Dynamic(), Get Mat Instance Dynamic()], ParameterName=ClothSim_R1_TurbulenceStrength, Value=@n6)
   > SetScalarParameterValue(self=[Get Mat Instance Dynamic(), Get Overlay Mat Instance Dynamic()], ParameterName=ClothSim_R2_TurbulenceStrength, Value=@n6)
   Set WPO Pho Wind Intensity(execute=@n4.else, WPO Pho Wind Intensity=0)
   > n7 = Set WPO Pho Turbulence(WPO Pho Turbulence=0)
   Outputs!(then_1:Exec=@n2.then_1)
  ]
 }
]
variables [
 AlphaCans {type=number}
 AlphaCloth {type=number}
 BeltDynAlpha {type=number}
 InSky? {type=boolean}
 isSurfing? {type=boolean}
 Local Noise Offset {type=Vector}
 Local Noise Transform Location {type=Vector}
 Mat Instance Dynamic {type=MaterialInstanceDynamic}
 Overlay Mat Instance Dynamic {type=MaterialInstanceDynamic}
 PlayerPawn {type=FortPlayerPawnAthena}
 R1 Velocity Transform Location {type=Vector}
 R1 Velocity Transform Velocity {type=Vector}
 R2 Velocity Transform Location {type=Vector}
 R2 Velocity Transform Velocity {type=Vector}
 SkelMesh {type=SkeletalMeshComponent}
 Velocity Interp Speed {type=number}
 WPO Pho Turbulence {type=number, default=4.2}
 WPO Pho Wind Intensity {type=number, default=0.8}
]
