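# Physics asset for the legs mesh of the FermiumStraw Hay medium female player body cosmetic.
# It approximates the lower body with one capsule each on the pelvis, both thighs and both calves, leaving the feet without bodies.
# Four constraints chain each calf to its thigh and each thigh to the pelvis, all sharing identical swing and twist limit angles rather than joint-specific ranges.
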PhysicsAsset /BRCosmetics/Characters/Player/Female/Medium/Bodies/F_MED_FermiumStraw_Hay/Meshes/F_MED_FermiumStraw_Hay_Legs_PhysicsAsset
bodies {
 calf_l [{shapeType=Capsule, length=30.2488, radius=12.2737, center=<-25.6898 1.0029 -1.3383>, rotation=<rot 0 80.7685 -89.9686>}]
 calf_r [{shapeType=Capsule, length=30.0313, radius=12.2719, center=<25.8068 -1.1455 0.5001>, rotation=<rot 0 -98.9776 -89.6213>}]
 pelvis [{shapeType=Capsule, length=17.8998, radius=12.9375, center=<-2.4982 2.2738 -0.0318>, rotation=<rot -13.9388 -90 -5.2671>}]
 thigh_l [{shapeType=Capsule, length=23.763, radius=12.3252, center=<-21.5209 -1.2429 -0.9283>, rotation=<rot 0 96.7336 -86.9118>}]
 thigh_r [{shapeType=Capsule, length=24.7308, radius=14.6847, center=<21.8045 0.2686 1.8657>, rotation=<rot 0 -75.1723 -81.9067>}]
}
constraints [
 {bone1Name=thigh_l, bone2Name=pelvis, swing1LimitDegrees=45, swing2LimitDegrees=45, twistLimitDegrees=45}
 {bone1Name=calf_l, bone2Name=thigh_l, swing1LimitDegrees=45, swing2LimitDegrees=45, twistLimitDegrees=45}
 {bone1Name=thigh_r, bone2Name=pelvis, swing1LimitDegrees=45, swing2LimitDegrees=45, twistLimitDegrees=45}
 {bone1Name=calf_r, bone2Name=thigh_r, swing1LimitDegrees=45, swing2LimitDegrees=45, twistLimitDegrees=45}
]
