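# Post-process style AnimBlueprint for the FermiumStraw Hay female torso cosmetic, adding secondary motion and bone corrections.
# It copies the attached parent's pose, runs the torso Control Rig, then chains spring-based AnimDynamics on jacket zipper, laces, hoodie and spray-can holders.
# Constant additive offsets then adjust upper-arm twist and clavicle corrective bones.
# The event graph sets front-end-dependent cloth/can alphas and a hang-glide flag, none visibly read by the AnimGraph.
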AnimBlueprint /BRCosmetics/Characters/Player/Female/Medium/Bodies/F_MED_FermiumStraw_Hay/Meshes/F_MED_FermiumStraw_Hay_Torso_AnimBP
target_skeleton=&/F_MED_FermiumStraw_Hay_Skeleton
graphs [
 AnimGraph {nodes [
   Output Pose(Result=Use cached pose 'ModifyBonesCache'()) {layerGroup=None}
   n1 = Control Rig(Source=Copy Pose From Mesh() {bUseAttachedParent=true}) {controlRigAssetReference {controlRigAsset=&/F_MED_FermiumStraw_Hay_Torso_CtrlRig}}
   Save cached pose 'ControlRigCache'(Pose=@n1) {cacheName=ControlRigCache}
   n2 = Anim Dynamics - Bone: dyn_jacketZipper!(ComponentPose=Local To Component(LocalPose=Use cached pose 'ControlRigCache'()), Alpha=0.6, GravityOverride:Vector="0,0,-980") -> (Pose) {
    angularDampingOverride=1
    angularSpringConstant=9000
    bAngularSpring=true
    bOverrideAngularDamping=true
    bUseGravityOverride=true
    gravityScale=0.35
    simulationSpace=RootRelative
    boundBone {boneName=dyn_jacketZipper}
    componentAppliedLinearAccClamp=<3000 3000 3000>
    componentLinearAccScale=<0.4 0.4 0.4>
    componentLinearVelScale=<0.4 0.4 0.4>
    gravityOverride {z=-980}
    physicsBodyDefinitions [{
      boundBone {boneName=dyn_jacketZipper}
      boxExtents=<8 20 8>
      constraintSetup {angularLimitsMax=<0 20 0>, angularLimitsMin=<0 -20 -70>, angularTarget=<1 0 0>}
      localJointOffset=<6 0 0>
     }]
    simSpaceSettings {maxAngularAcceleration=20, maxAngularVelocity=10, simSpaceAngularAlpha=1}
   }
   n3 = Anim Dynamics - Bone: dyn_jacketLace_01_l!(ComponentPose=@n2, Alpha=0.6, GravityOverride:Vector="0,0,-980") -> (Pose) {
    angularDampingOverride=0.9
    angularSpringConstant=50
    bAngularSpring=true
    bOverrideAngularDamping=true
    bUseGravityOverride=true
    gravityScale=0.35
    simulationSpace=RootRelative
    boundBone {boneName=dyn_jacketLace_01_l}
    componentAppliedLinearAccClamp=<3000 3000 3000>
    componentLinearAccScale=<0.4 0.4 0.4>
    componentLinearVelScale=<0.2 0.2 0.2>
    gravityOverride {z=-980}
    physicsBodyDefinitions [{
      boundBone {boneName=dyn_jacketLace_01_l}
      boxExtents=<15 15 15>
      constraintSetup {angularLimitsMax=<0 40 0>, angularLimitsMin=<0 -40 -90>, angularTarget=<1 0 0>}
      localJointOffset=<5 0 0>
     }]
    simSpaceSettings {maxAngularAcceleration=20, maxAngularVelocity=10, simSpaceAngularAlpha=1}
   }
   n4 = Anim Dynamics - Bone: dyn_jacketLace_02_l!(ComponentPose=@n3, Alpha=0.6, GravityOverride:Vector="0,0,-980") -> (Pose) {
    angularDampingOverride=1
    angularSpringConstant=50
    bAngularSpring=true
    bOverrideAngularDamping=true
    bUseGravityOverride=true
    gravityScale=0.35
    simulationSpace=RootRelative
    boundBone {boneName=dyn_jacketLace_02_l}
    componentAppliedLinearAccClamp=<3000 3000 3000>
    componentLinearAccScale=<5 5 5>
    componentLinearVelScale=<0.4 0.4 0.4>
    gravityOverride {z=-980}
    physicsBodyDefinitions [{
      boundBone {boneName=dyn_jacketLace_02_l}
      boxExtents=<20 20 20>
      constraintSetup {angularLimitsMax=<0 40 0>, angularLimitsMin=<0 -40 -90>, angularTarget=<1 0 0>}
      localJointOffset=<4 0 0>
     }]
    simSpaceSettings {maxAngularAcceleration=20, maxAngularVelocity=10, simSpaceAngularAlpha=1}
   }
   n5 = Anim Dynamics - Bone: dyn_jacketLace_01_r!(ComponentPose=@n4, Alpha=0.6, GravityOverride:Vector="0,0,-980") -> (Pose) {
    angularDampingOverride=0.9
    angularSpringConstant=50
    bAngularSpring=true
    bOverrideAngularDamping=true
    bUseGravityOverride=true
    gravityScale=0.35
    simulationSpace=RootRelative
    boundBone {boneName=dyn_jacketLace_01_r}
    componentAppliedLinearAccClamp=<3000 3000 3000>
    componentLinearAccScale=<0.4 0.4 0.4>
    componentLinearVelScale=<0.2 0.2 0.2>
    gravityOverride {z=-980}
    physicsBodyDefinitions [{
      boundBone {boneName=dyn_jacketLace_01_r}
      boxExtents=<20 20 20>
      constraintSetup {angularLimitsMax=<0 40 0>, angularLimitsMin=<0 -40 -90>, angularTarget=<1 0 0>}
      localJointOffset=<-4 0 0>
     }]
    simSpaceSettings {maxAngularAcceleration=20, maxAngularVelocity=10, simSpaceAngularAlpha=1}
   }
   n6 = Anim Dynamics - Bone: dyn_jacketLace_02_r!(ComponentPose=@n5, Alpha=0.6, GravityOverride:Vector="0,0,-980") -> (Pose) {
    angularDampingOverride=1
    angularSpringConstant=50
    bAngularSpring=true
    bOverrideAngularDamping=true
    bUseGravityOverride=true
    gravityScale=0.35
    simulationSpace=RootRelative
    boundBone {boneName=dyn_jacketLace_02_r}
    componentAppliedLinearAccClamp=<3000 3000 3000>
    componentLinearAccScale=<0.4 0.4 0.4>
    componentLinearVelScale=<0.4 0.4 0.4>
    gravityOverride {z=-980}
    physicsBodyDefinitions [{
      boundBone {boneName=dyn_jacketLace_02_r}
      boxExtents=<20 20 20>
      constraintSetup {angularLimitsMax=<0 40 0>, angularLimitsMin=<0 -40 -90>, angularTarget=<1 0 0>}
      localJointOffset=<-6 0 0>
     }]
    simSpaceSettings {maxAngularAcceleration=20, maxAngularVelocity=10, simSpaceAngularAlpha=1}
   }
   n7 = Anim Dynamics - Bone: dyn_hoodie!(ComponentPose=@n6, Alpha=0.3, GravityOverride:Vector="0,0,-980") -> (Pose) {
    angularDampingOverride=1
    angularSpringConstant=9000
    bAngularSpring=true
    bOverrideAngularDamping=true
    bUseGravityOverride=true
    gravityScale=0.35
    simulationSpace=RootRelative
    boundBone {boneName=dyn_hoodie}
    componentAppliedLinearAccClamp=<3000 3000 3000>
    componentLinearAccScale=<0.6 0.6 0.6>
    componentLinearVelScale=<0.2 0.2 0.2>
    gravityOverride {z=-980}
    physicsBodyDefinitions [{
      boundBone {boneName=dyn_hoodie}
      boxExtents=<20 20 20>
      constraintSetup {angularLimitsMax=<0 20 10>, angularLimitsMin=<0 -20 -50>, angularTarget=<1 0 0>}
      localJointOffset=<4 0 0>
     }]
    simSpaceSettings {maxAngularAcceleration=20, maxAngularVelocity=10, simSpaceAngularAlpha=1}
   }
   n8 = Anim Dynamics - Bone: dyn_sprayCanHolder_l!(ComponentPose=@n7, Alpha=0.7, GravityOverride:Vector="0,0,-980") -> (Pose) {
    angularDampingOverride=1
    angularSpringConstant=9000
    bAngularSpring=true
    bOverrideAngularDamping=true
    bUseGravityOverride=true
    gravityScale=0.35
    simulationSpace=RootRelative
    boundBone {boneName=dyn_sprayCanHolder_l}
    componentAppliedLinearAccClamp=<3000 3000 3000>
    componentLinearAccScale=<0.4 0.4 0.4>
    componentLinearVelScale=<0.2 0.2 0.2>
    gravityOverride {z=-980}
    physicsBodyDefinitions [{
      boundBone {boneName=dyn_sprayCanHolder_l}
      boxExtents=<20 20 20>
      constraintSetup {angularLimitsMax=<40 20 20>, angularLimitsMin=<-20 -20 -70>, angularTarget=<1 0 0>}
      localJointOffset=<2 0 0>
     }]
    simSpaceSettings {maxAngularAcceleration=20, maxAngularVelocity=10, simSpaceAngularAlpha=1}
   }
   n9 = Anim Dynamics - Bone: dyn_sprayCanHolder_r!(ComponentPose=@n8, Alpha=0.7, GravityOverride:Vector="0,0,-980") -> (Pose) {
    angularDampingOverride=1
    angularSpringConstant=9000
    bAngularSpring=true
    bOverrideAngularDamping=true
    bUseGravityOverride=true
    gravityScale=0.35
    simulationSpace=RootRelative
    boundBone {boneName=dyn_sprayCanHolder_r}
    componentAppliedLinearAccClamp=<3000 3000 3000>
    componentLinearAccScale=<0.4 0.4 0.4>
    componentLinearVelScale=<0.2 0.2 0.2>
    gravityOverride {z=-980}
    physicsBodyDefinitions [{
      boundBone {boneName=dyn_sprayCanHolder_r}
      boxExtents=<16 16 16>
      constraintSetup {angularLimitsMax=<20 20 20>, angularLimitsMin=<-40 -20 -40>, angularTarget=<1 0 0>}
      localJointOffset=<3 0 0>
     }]
    simSpaceSettings {maxAngularAcceleration=20, maxAngularVelocity=10, simSpaceAngularAlpha=1}
   }
   Save cached pose 'AnimDynamicsCache'(Pose=Component To Local(ComponentPose=@n9)) {cacheName=AnimDynamicsCache}
   n10 = "Transform (Modify) Bone - Bone: upperarm_twist_01_l"!(ComponentPose=Local To Component(LocalPose=Use cached pose 'AnimDynamicsCache'()), Alpha=1, Rotation="0,0,-50") -> (Pose) {
    rotationMode=BMM_Additive
    rotationSpace=BCS_BoneSpace
    scaleMode=BMM_Additive
    translationMode=BMM_Additive
    boneToModify {boneName=upperarm_twist_01_l}
    rotation {roll=-50}
   }
   n11 = "Transform (Modify) Bone - Bone: upperarm_twist_02_l"!(ComponentPose=@n10, Alpha=1, Rotation="0,0,-25") -> (Pose) {
    rotationMode=BMM_Additive
    rotationSpace=BCS_BoneSpace
    scaleMode=BMM_Additive
    translationMode=BMM_Additive
    boneToModify {boneName=upperarm_twist_02_l}
    rotation {roll=-25}
   }
   n12 = "Transform (Modify) Bone - Bone: upperarm_twist_01_r"!(ComponentPose=@n11, Alpha=1, Rotation="0,0,-50") -> (Pose) {
    rotationMode=BMM_Additive
    rotationSpace=BCS_BoneSpace
    scaleMode=BMM_Additive
    translationMode=BMM_Additive
    boneToModify {boneName=upperarm_twist_01_r}
    rotation {roll=-50}
   }
   n13 = "Transform (Modify) Bone - Bone: upperarm_twist_02_r"!(ComponentPose=@n12, Alpha=1, Rotation="0,0,-25") -> (Pose) {
    rotationMode=BMM_Additive
    rotationSpace=BCS_BoneSpace
    scaleMode=BMM_Additive
    translationMode=BMM_Additive
    boneToModify {boneName=upperarm_twist_02_r}
    rotation {roll=-25}
   }
   n14 = "Transform (Modify) Bone - Bone: cor_clavicle_r"!(ComponentPose=@n13, Alpha=1, Translation="0,2,-2", Rotation="0,0,0") -> (Pose) {
    rotationMode=BMM_Additive
    rotationSpace=BCS_BoneSpace
    scaleMode=BMM_Additive
    translationMode=BMM_Additive
    boneToModify {boneName=cor_clavicle_r}
    translation {y=2, z=-2}
   }
   n15 = "Transform (Modify) Bone - Bone: cor_clavicle_l"!(ComponentPose=@n14, Alpha=1, Translation="0,2,-2", Rotation="0,0,0") -> (Pose) {
    rotationMode=BMM_Additive
    rotationSpace=BCS_BoneSpace
    scaleMode=BMM_Additive
    translationMode=BMM_Additive
    boneToModify {boneName=cor_clavicle_l}
    translation {y=2, z=-2}
   }
   Save cached pose 'ModifyBonesCache'(Pose=Component To Local(ComponentPose=@n15)) {cacheName=ModifyBonesCache}
  ]}
 EventGraph {nodes [
   Event OnParachuteOpened()
   > Switch on EGliderType(Selection=HangGlider)
   > Set isHangGliding(isHangGliding:bool=true) -> (Output_Get:bool)
   Event OnParachuteClosed()
   > Set isHangGliding(isHangGliding=false)
   Event BlueprintUpdateAnimation()
   > n1 = Branch(Condition=Get bIsInFrontEnd())
   > Set AlphaCloth(AlphaCloth:double=0.4) -> (Output_Get:double)
   > Set AlphaCans(AlphaCans:double=0.35) -> (Output_Get:double)
   Set AlphaCloth(execute=@n1.else, AlphaCloth=0.6)
   > Set AlphaCans(AlphaCans=0.7)
  ]}
]
variables [
 AlphaCans {type=number}
 AlphaCloth {type=number}
 isHangGliding {type=boolean}
]
